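# Niagara swing trail for the MuteRibbon pickaxe, layering short-lived ribbons with mesh music notes.
# AirTrail and two DarkStreaks emitters use shared AnimTrail modules, fading over particle life, with AirTrail also narrowing.
# Music_Notes spawns meshes by distance moved at a socket on the user skeletal mesh, scaling them over life, with a local module gating orientation updates by age.
# All four emitters take their colors from time-of-day inputs.
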NiagaraSystem /BRCosmetics/Weapons/FORT_Melee/Pickaxe_MuteRibbon/FX/NS_Pickaxe_MuteRibbon_Trail
dependencies {
 usedDataInterfaces [
  &NiagaraDataInterfaceCurve
  &NiagaraDataInterfaceColorCurve
  &NiagaraDataInterfaceMeshRendererInfo
 ]
 usedDynamicInputs []
 usedModules [&SystemState]
}
emitters {
 AirTrail {
  emitterUpdateScript [
   &EmitterState()
   &FN_SpawnAnimTrail_v2(AnimTrailWidthMode=::FromFirst)
  ]
  particleSpawnScript [
   /Game/Effects/Niagara/Modules/AnimTrail/FN_InitializeAnimTrail/FN_InitializeAnimTrailBoneIndecies(First Bone or Socket Index=1)
   /Game/Effects/Niagara/Modules/AnimTrail/FN_InitializeAnimTrail/FN_InitializeAnimTrail(
    AnimTrailWidth=1
    Color=&FNTimeOfDayColor(
     Night Color=<rgba 0.1354 0.0935 0 1>
     Morning Color=<rgba 0.4115 0.2123 0 1>
     Day Color=<rgba 0.4115 0.2123 0 1>
     Evening Color=<rgba 0.4115 0.2123 0 1>
    )
    Lifetime=0.1135
   )
  ]
  particleUpdateScript [
   &DynamicMaterialParameters(
    Index 0 Param 1=&UniformRangedFloat(Minimum=0, Maximum=1)
    Index 0 Param 2=&UniformRangedFloat(Minimum=0, Maximum=1)
    Index 0 Param 3=1
    Index 0 Param 4=1
    Module.Param0WriteEnabled=true
    Module.Param1WriteEnabled=true
    Module.Param2WriteEnabled=true
    Module.Param3WriteEnabled=true
   )
   &FN_UpdateAnimTrail_v2()
   &ScaleRibbonWidth(Ribbon Width Scale=&FloatFromCurve(FloatCurve=DataInterface() {propertyValues {Curve {keys [
         {arriveTangent=-1.2579, interpMode=RCIM_Cubic, time=0, value=1}
         {arriveTangent=-0.9664, interpMode=RCIM_Cubic, leaveTangent=-0.9664, tangentMode=RCTM_User, time=1, value=0}
        ]}}}))
   &ScaleColor(
    Scale Alpha=&FloatFromCurve(FloatCurve=DataInterface() {propertyValues {Curve {keys [{interpMode=RCIM_Cubic, tangentMode=RCTM_User, time=0, value=1}, {interpMode=RCIM_Cubic, tangentMode=RCTM_User, time=1, value=0}]}}})
    Scale RGB=&MakeVectorFromLinearColorRGB(Color=&ColorFromCurve(ColorCurve=DataInterface() {propertyValues {
        AlphaCurve {keys [
          {arriveTangent=-1.791, interpMode=RCIM_Cubic, leaveTangent=-1.791, time=0.4417, value=1}
          {arriveTangent=-1.791, interpMode=RCIM_Cubic, leaveTangent=-1.791, time=1, value=0}
         ]}
        BlueCurve {keys [<at 0 1>, {interpMode=RCIM_Cubic, leaveTangent=-0.8114, time=0.6162, value=0.6354}]}
        GreenCurve {keys [<at 0 1>, {interpMode=RCIM_Cubic, leaveTangent=-0.8114, time=0.6162, value=0.6354}]}
        RedCurve {keys [<at 0 1>, {interpMode=RCIM_Cubic, leaveTangent=-0.8114, time=0.6162, value=0.6354}]}
       }}))
    ScaleA=true
    ScaleRGB=true
   )
  ]
  rendererClasses [&NiagaraRibbonRendererProperties]
  renderers [
   {rendererClass=&NiagaraRibbonRendererProperties}
   {rendererClass=&NiagaraRibbonRendererProperties, rendererIndex=1}
  ]
 }
 DarkStreaks {
  emitterUpdateScript [
   &EmitterState()
   &FN_SpawnAnimTrail_v2()
  ]
  particleSpawnScript [
   &FN_InitializeAnimTrailBoneIndecies(First Bone or Socket Index=1)
   &FN_InitializeAnimTrail(
    AnimTrailWidth=1
    Color=&FNTimeOfDayColor(
     Night Color=<rgba 0.099 0.0709 0 6>
     Morning Color=<rgba 0.3021 0.1361 0 6>
     Day Color=<rgba 0.3712 0.1559 0 6>
     Evening Color=<rgba 0.3021 0.1361 0 6>
    )
    Lifetime=0.1
    Write Material Random=false
   )
  ]
  particleUpdateScript [
   &ScaleColor(
    Scale Alpha=&FloatFromCurve(FloatCurve=DataInterface() {propertyValues {Curve {keys [
         {arriveTangent=55.0714, interpMode=RCIM_Cubic, leaveTangent=55.0714, tangentMode=RCTM_User, time=0.1, value=0}
         {arriveTangent=-1.1764, interpMode=RCIM_Cubic, leaveTangent=-1.1764, tangentMode=RCTM_User, time=0.1408, value=0.9847}
         {arriveTangent=-0.0405, interpMode=RCIM_Cubic, leaveTangent=-0.0405, tangentMode=RCTM_User, time=1, value=0}
        ]}}})
    ScaleA=true
    ScaleRGB=true
   )
   &FN_UpdateAnimTrail_v2()
  ]
  rendererClasses [&NiagaraRibbonRendererProperties]
  renderers [{rendererClass=&NiagaraRibbonRendererProperties}]
 }
 DarkStreaks001 {
  emitterUpdateScript [
   &EmitterState()
   &FN_SpawnAnimTrail_v2()
  ]
  particleSpawnScript [
   &FN_InitializeAnimTrailBoneIndecies(First Bone or Socket Index=1)
   &FN_InitializeAnimTrail(AnimTrailWidth=0.45, Lifetime=0.15, Write Material Random=false)
  ]
  particleUpdateScript [
   &FN_UpdateAnimTrail_v2()
   &FN_RibbonMaterial(
    Color=&FNTimeOfDayColor(
     Night Color=<rgba 0.0938 0.0428 0.0188 4>
     Morning Color=<rgba 0.0156 0.0071 0.0031 4>
     Day Color=<rgba 0.0156 0.0071 0.0031 4>
     Evening Color=<rgba 0.0156 0.0071 0.0031 4>
    )
    Scale Alpha:NiagaraFloat=&FloatFromCurve(FloatCurve=DataInterface() {propertyValues {Curve {keys [
         {
          arriveTangent=-1.2957
          arriveTangentWeight=0.1148
          interpMode=RCIM_Cubic
          leaveTangent=-1.2957
          leaveTangentWeight=0.1148
          tangentMode=RCTM_User
          tangentWeightMode=RCTWM_WeightedBoth
          time=0
          value=0
         }
         {arriveTangent=-1.5958, interpMode=RCIM_Cubic, leaveTangent=-1.5958, tangentMode=RCTM_User, time=0.075, value=1}
         {interpMode=RCIM_Cubic, tangentMode=RCTM_User, time=1, value=0}
        ]}}})
   )
  ]
  rendererClasses [&NiagaraRibbonRendererProperties]
  renderers [{rendererClass=&NiagaraRibbonRendererProperties}]
 }
 Music_Notes {
  emitterUpdateScript [
   &EmitterState(Life Cycle Mode=::NewEnumerator1, Loop Delay=0.05, Loop Duration=0.15, UseLoopDelay=true)
   &SpawnPerUnit(Module.Use Max Movement Threshold=true, Module.Use Movement Tolerance=false, Spawn Spacing=18, Use Spawn Probability=false)
  ]
  particleSpawnScript [
   &InitializeParticle(
    Color=&FNTimeOfDayColor(
     Night Color=<rgba 0.526 0.2751 0 1>
     Morning Color=<rgba 0.9323 0.4875 0 1>
     Day Color=<rgba 0.9323 0.4875 0 1>
     Evening Color=<rgba 0.9323 0.4875 0 1>
    )
    Lifetime=0.1
    Mesh Renderer Array Visibility Mode=::NewEnumerator3
    Mesh Renderer Info=DataInterface() {propertyValues {MeshRenderer=&self:Music_Notes_0.NiagaraMeshRendererProperties_0}}
    Mesh Scale Mode=::NewEnumerator3
    Mesh Uniform Scale=1.2
    UsePositionOffset=false
    Write Material Random=true
   )
   &SkeletalMeshLocation(
    Bone / Socket Index=2
    MaskSpawnGroup=false
    Mesh Sampling Type=::NewEnumerator5
    Sampled Position Offset=&RandomRangeVector(Minimum=<0 0 -8>, Maximum=<0 0 5>)
    Skeletal Mesh=$User.SkeletalMesh
    Socket Sampling Mode=::NewEnumerator5
   )
  ]
  particleUpdateScript [
   &ParticleState()
   &SkeletalMeshLocation(
    Bone / Socket Index=2
    MaskSpawnGroup=false
    Mesh Sampling Type=::NewEnumerator5
    Skeletal Mesh=$User.SkeletalMesh
    Socket Sampling Mode=::NewEnumerator5
    Write To Particle Position=false
   ) {module=SkeletalMeshLocation001}
   /SetVariables("Particles.Stopped Mesh Orientation ":Quat4f=$Particles.SkeletalMeshLocation001.SampledRotation) {isSetParameters=true}
   &self:PreviousOrientation(
    Stop Update:NiagaraBool=&SetBoolByFloatComparison(A=$Particles.NormalizedAge, B=0.2, Comparison Type=::NewEnumerator2)
    Updating Mesh Orientation:Quat4f=$Particles.SkeletalMeshLocation001.SampledRotation
   ) {module=PreviousOrientation}
   &ScaleMeshSize(Scale Factor=&VectorFromFloat(Value=&FloatFromCurve(FloatCurve=DataInterface() {propertyValues {Curve {keys [
          {arriveTangent=9.0725, interpMode=RCIM_Cubic, leaveTangent=9.0725, tangentMode=RCTM_User, time=0, value=0}
          <at 0.025 1.2 RCIM_Cubic>
          {arriveTangent=-2.4073, interpMode=RCIM_Cubic, leaveTangent=-2.4073, tangentMode=RCTM_User, time=1, value=0}
         ]}}})))
  ]
  rendererClasses [&NiagaraMeshRendererProperties]
  renderers [{rendererClass=&NiagaraMeshRendererProperties, rendererIndex=0}]
 }
}
summary {
 systemName=NS_Pickaxe_MuteRibbon_Trail
 emitters [
  {emitterName=DarkStreaks, rendererClasses [&NiagaraRibbonRendererProperties]}
  {emitterName=AirTrail, rendererClasses [&NiagaraRibbonRendererProperties]}
  {emitterName=Music_Notes, rendererClasses [&NiagaraMeshRendererProperties]}
  {emitterName=DarkStreaks001, rendererClasses [&NiagaraRibbonRendererProperties]}
 ]
 userVariables [
  User.bIsFrontEnd {type=NiagaraBool}
  User.RibbonIDOffset {type=NiagaraInt32}
  User.SkeletalMesh {
   type=NiagaraDataInterfaceSkeletalMesh
   defaultValue {
    struct=NiagaraExt_VariableValue_DataInterface
    value {dataInterface=&self:NiagaraDataInterfaceSkeletalMesh_0}
   }
  }
  User.SwingSectionIndex {type=NiagaraInt32}
 ]
}
